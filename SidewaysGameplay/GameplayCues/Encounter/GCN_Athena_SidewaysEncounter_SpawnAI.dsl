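; This gameplay cue notify plays the burst effect for an AI spawn in a Sideways encounter, apparently a vertical lightning strike.
; On burst, it takes the cue location and derives a second point high above it.
; If the first Cascade particle component and the target are valid, it moves that component to the location, aimed at the upper point.
; The scaled distance between the points sets the particle's LightningLength parameter.
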(fn OnBurst(Cascade)
  (params (MyTarget :type Actor%20Object%20Reference) (Parameters :type Gameplay%20Cue%20Parameters%20Structure%20%28by%20ref%29) (ParticleComponents :type Array%20of%20Cascade%20Particle%20System%20Component%20Object%20References) (AudioComponents :type Array%20of%20Audio%20Component%20Object%20References) (BurstCameraShakeInstance :type Camera%20Shake%20Base%20Object%20Reference) (BurstDecalInstance :type Decal%20Actor%20Object%20Reference))
  (bind (_normalizedmagnitude _rawmagnitude _effectcontext _matchedtagname _originaltag _aggregatedsourcetags _aggregatedtargettags _location) (Ability|GameplayCue|BreakGameplayCueParameters Parameters))
  (bind _returnvalue (+ _location "0, 0,5000.000000"))
  (bind _output (Utilities|Array|Get(acopy) ParticleComponents))
  (Utilities|IsValid _output
    (:Is%20Valid
      (Utilities|IsValid MyTarget
        (:Is%20Valid
          (Transformation|SetWorldLocationAndRotation _output _location (Math|Rotator|FindLookatRotation _location _returnvalue) false false)
          (Effects|Components|ParticleSystem|SetVectorParameter _output "LightningLength" (Math|Vector|MakeVector (/ (Math|Vector|VectorLength (- _returnvalue _location)) 1e+03) 1.0 1.0)))
        (:Is%20Not%20Valid
          _))
      (return))
    (:Is%20Not%20Valid
      _)))
